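# Translucent, unlit, two-sided Niagara mesh-particle material, named as a lens speed-cone effect.
# Panning high-contrast noise distorts a V-axis sphere mask, forming a soft band tinted by particle color and boosted to HDR.
# An opacity chain combines that band with gradient masks, inverse Fresnel, near-camera fade, particle alpha and depth fade.
# The noise also drives a displacement along vertex normals; an axis-aligned Fresnel branch feeds neither color nor opacity chain.
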Material /NitroGameplay/Assets/TEMP/FX/M_Lens_Speed_Cone
graph {
 properties {
  bUsedWithNiagaraMeshParticles=true
  blendMode=BLEND_Translucent
  pixelDepthOffsetMode=PDOM_Legacy
  shadingModel=MSM_Unlit
  twoSided=true
  usedShadingModels=Unlit
 }
 nodes [
  Add()
  n1 = Power(Base=ComponentMask(None=TextureCoordinate()) {g=true}, Exp=ScalarParameter() {defaultValue=0.66, parameterName=Param_8})
  n2 = TextureSample(UVs=Panner(Coordinate=TextureCoordinate() {uTiling=3}) {speedY=-1.75}) {samplerType=SAMPLERTYPE_LinearColor, texture=&HighContrastNoise}
  n3 = ScalarParameter() {defaultValue=0.5, parameterName=Param_1}
  n4 = ComponentMask(None=TextureCoordinate()) {g=true}
  n5 = SphereMask(A=Add(A=Multiply(A=@n2.G, B=@n3), B=@n4), B=ScalarParameter() {defaultValue=0.25}) {attenuationRadius=0.5, hardnessPercent=0}
  n6 = Multiply(
   A=@n5
   B=SphereMask(
    A=@n4
    B=ScalarParameter() {defaultValue=0.5, parameterName=Param_3}
    Radius=ScalarParameter() {defaultValue=0.5, parameterName=Param_4}
    Hardness=ScalarParameter() {parameterName=Param_5}
   )
  )
  n7 = ScalarParameter() {defaultValue=2, parameterName=Param_7}
  n8 = Fresnel() {baseReflectFraction=0, exponent=1}
  n9 = MaterialFunctionCall(Distance=ScalarParameter() {defaultValue=500, parameterName=Param_10}) {materialFunction=&Near_Camera_Fade}
  n10 = Multiply(A=Multiply(A=Multiply(A=@n6, B=@n7), B=OneMinus(None=@n8)), B=@n9.Result)
  n11 = ParticleColor()
  DepthFade(Opacity=Multiply(A=Multiply(A=@n1, B=@n10), B=@n11.A))
  n12 = Multiply(A=Multiply(A=@n5, B=@n11.RGB), B=Multiply(A=@n1, B=@n9.Result))
  MaterialFunctionCall(LDR Input=Saturate(None=@n12), HDR Tint=@n12, HDR Intensity Multiplier=ScalarParameter() {defaultValue=2, parameterName=Param_2}) {materialFunction=&DeriveHDRfromLDR}
  n13 = ScalarParameter() {defaultValue=15, parameterName=Param_9}
  Multiply(A=LinearInterpolate(A=Multiply(A=@n13) {constB=-1}, B=@n13, Alpha=@n2.G) {constA=-10, constB=10}, B=VertexNormalWS())
  Multiply()
  OneMinus()
  OneMinus()
  Power(Base=@n4) {constExponent=0.66}
  n14 = Transform(None=VectorParameter() {defaultValue=<rgba 1 0 0 0>}.RGB) {transformSourceType=TRANSFORMSOURCE_Local}
  Power(
   Base=MaterialFunctionCall(Axis=Normalize(VectorInput=@n14)) {materialFunction=&AxisAlignedFresnel}.Fresnel
   Exp=ScalarParameter() {defaultValue=2, parameterName=Param_6}
  )
 ]
}
